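# HUD mic indicator widget, derived from FortMicIndicatorWidget, that shows a player's voice state as one of three icons.
# A WidgetSwitcher holds overlays for not speaking, speaking and muted, each with its own sound icon.
# The parent class's muted and talking-changed events cache their flags and drive the switcher index.
# Mute takes priority: talking changes leave the switcher untouched while muted, and unmuting restores the icon from the cached talking flag.
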WidgetBlueprint /Game/UI/InGame/HUD/MicIndicator/MicIndicator
graphs [EventGraph {nodes [
   n1 = Event OnPlayerMuted!() -> (OutputDelegate, then, bIsMuted:bool) {is=title K2Node_Event}
   > n2 = Set bIsMuted(bIsMuted:bool=@n1.bIsMuted) -> (Output_Get:bool)
   > n3 = Branch(Condition=@n2)
   n4 = Get MicStateSwitcher() -> (MicStateSwitcher:Widget Switcher Object Reference)
   > SetActiveWidgetIndex(self=@n4, Index=2)
   SetActiveWidgetIndex(execute=[@n6.else, @n3.else], self=@n4, Index=SelectInt(A=1, B=0, bPickA=Get bIsTalking() -> (bIsTalking:bool)))
   n5 = Event OnPlayerTalkingChanged!() -> (OutputDelegate, then, bIsTalking:bool) {is=title K2Node_Event}
   > Set bIsTalking(bIsTalking:bool=@n5.bIsTalking) -> (Output_Get:bool)
   > n6 = Branch(Condition=Get bIsMuted() -> (bIsMuted:bool))
  ]}]
variables [
 bIsMuted {type=boolean}
 bIsTalking {type=boolean}
]
widgets {
 info {inheritedWidgetCount=0, parentClass=/Script/FortniteUI.FortMicIndicatorWidget, rootWidgetClass=&WidgetSwitcher, widgetCount=7}
 widgets [
  MicStateSwitcher {widgetClassPath=&WidgetSwitcher, bIsVariable=true, widget=MicStateSwitcher}
  01-NotSpeaking {
   widgetClassPath=&Overlay
   bIsVariable=true
   parent=MicStateSwitcher
   widget=01-NotSpeaking
   slot {class=WidgetSwitcherSlot, properties {horizontalAlignment=HAlign_Center, verticalAlignment=VAlign_Center}}
  }
  Image_13 {
   widgetClassPath=&Image
   parent=01-NotSpeaking
   widget=Image_13
   properties {brush {
     bIsDynamicallyLoaded=false
     drawAs=Image
     imageType=NoImage
     mirroring=NoMirror
     resourceName=None
     resourceObject=/Game/UI/Foundation/Textures/Icons/Audio/T-Icon-Sound-M
     tiling=NoTile
     imageSize=<32 32>
     margin=<ltrb 0 0 0 0>
     outlineSettings {
      bUseBrushTransparency=false
      roundingType=HalfHeightRadius
      width=0
      color {colorUseRule=UseColor_Specified, specifiedColor=<rgba 0 0 0 0>}
      cornerRadii=<0 0 0 0>
     }
     tintColor {colorUseRule=UseColor_Specified, specifiedColor=<rgba 1 1 1 1>}
     uVRegion {bIsValid=false, max=<0 0>, min=<0 0>}
    }}
   slot {class=OverlaySlot, properties {horizontalAlignment=HAlign_Center, verticalAlignment=VAlign_Center}}
  }
  02-Speaking {
   widgetClassPath=&Overlay
   bIsVariable=true
   parent=MicStateSwitcher
   widget=02-Speaking
   slot {class=WidgetSwitcherSlot, properties {horizontalAlignment=HAlign_Center, verticalAlignment=VAlign_Center}}
  }
  Image_14 {
   widgetClassPath=&Image
   parent=02-Speaking
   widget=Image_14
   properties {brush {
     bIsDynamicallyLoaded=false
     drawAs=Image
     imageType=NoImage
     mirroring=NoMirror
     resourceName=None
     resourceObject=/Game/UI/Foundation/Textures/Icons/Audio/T-Icon-Sound-On-M
     tiling=NoTile
     imageSize=<32 32>
     margin=<ltrb 0 0 0 0>
     outlineSettings {
      bUseBrushTransparency=false
      roundingType=HalfHeightRadius
      width=0
      color {colorUseRule=UseColor_Specified, specifiedColor=<rgba 0 0 0 0>}
      cornerRadii=<0 0 0 0>
     }
     tintColor {colorUseRule=UseColor_Specified, specifiedColor=<rgba 1 1 1 1>}
     uVRegion {bIsValid=false, max=<0 0>, min=<0 0>}
    }}
   slot {class=OverlaySlot, properties {horizontalAlignment=HAlign_Center, verticalAlignment=VAlign_Center}}
  }
  03-Muted {
   widgetClassPath=&Overlay
   bIsVariable=true
   parent=MicStateSwitcher
   widget=03-Muted
   slot {class=WidgetSwitcherSlot, properties {horizontalAlignment=HAlign_Center, verticalAlignment=VAlign_Center}}
  }
  Image_0 {
   widgetClassPath=&Image
   parent=03-Muted
   widget=Image_0
   properties {brush {
     bIsDynamicallyLoaded=false
     drawAs=Image
     imageType=NoImage
     mirroring=NoMirror
     resourceName=None
     resourceObject=/Game/UI/Foundation/Textures/Icons/Audio/T-Icon-Sound-Mute-M
     tiling=NoTile
     imageSize=<32 32>
     margin=<ltrb 0 0 0 0>
     outlineSettings {
      bUseBrushTransparency=false
      roundingType=HalfHeightRadius
      width=0
      color {colorUseRule=UseColor_Specified, specifiedColor=<rgba 0 0 0 0>}
      cornerRadii=<0 0 0 0>
     }
     tintColor {colorUseRule=UseColor_Specified, specifiedColor=<rgba 1 1 1 1>}
     uVRegion {bIsValid=false, max=<0 0>, min=<0 0>}
    }}
   slot {class=OverlaySlot, properties {horizontalAlignment=HAlign_Center, verticalAlignment=VAlign_Center}}
  }
 ]
}
