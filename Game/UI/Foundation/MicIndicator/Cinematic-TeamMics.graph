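# Team member row widget, apparently for cinematics, showing a mic indicator and player name, or "Connecting..." while the member's hero item is invalid.
# Update, also called on Construct, stores the member info, sets the name and picks the console net ID if it matches the local players' platform, otherwise the member ID.
# Only with a valid hero does it pass that ID to MicIndicator and set its own visibility from home base data validity.
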WidgetBlueprint /Game/UI/Foundation/MicIndicator/Cinematic-TeamMics
graphs [
 EventGraph {nodes [
   Event Construct()
   > Update!(execute, self, UpdatedMemberInfo:Fort Team Member Info Structure=Get MemberInfo() -> (MemberInfo:Fort Team Member Info Structure)) -> (then)
  ]}
 Update {nodes [
   n1 = Update!() -> (then, UpdatedMemberInfo:Fort Team Member Info Structure) {is=title K2Node_FunctionEntry}
   > Set MemberInfo(MemberInfo:Fort Team Member Info Structure=@n1) -> (Output_Get:Fort Team Member Info Structure)
   n2 = Break Fort Team Member Info!(FortTeamMemberInfo:"Fort Team Member Info Structure (by ref)"=Get MemberInfo()) -> (
    MemberUniqueId:Unique Net Id Structure
    PartyLeaderUniqueId:Unique Net Id Structure
    ConsoleUniqueId:Unique Net Id Structure
    Platform:String
    PlayerName:Text
    bPartyLeader:bool=~
    bIsInZone:bool=~
    bHasBoostXp:bool=~
    bHasRestXp:bool=~
    BannerIconId:String
    BannerColorId:String
    bBattlePassPurchased:bool=~
    BattlePassLevel:int
    NumPlayersInParty:int
    PlayerIndex:int
    TeamAffiliation:Byte
    HeroClass:Text
    HeroLevel:Text
    HeroXP:int
    HeroItem:Fort Item Object Reference
    SelectedGadgetItems:Array of Fort Item Object References
    HeroIcon:Slate Brush Structure
    HomeBaseInfo:Fort Home Base Info Structure
   ) {is=title K2Node_BreakStruct}
   n3 = Get FortPartyContext!() -> (ReturnValue:Fort Party Context Object Reference) {is=title K2Node_GetBlueprintContextEx}
   > Set PlayerNetId(PlayerNetId:Unique Net Id Structure=Select(
     Option 0:"Unique Net Id Structure (by ref)"=@n2
     Option 1:"Unique Net Id Structure (by ref)"=@n2.ConsoleUniqueId
     Index=IsNetIdFromLocalPlayersPlatform!(self:Fort Party Context Object Reference=@n3, PlayerId:"Unique Net Id Structure (by ref)"=@n2.ConsoleUniqueId) -> (ReturnValue:bool)
    ) -> (ReturnValue:"Unique Net Id Structure (by ref)")) -> (Output_Get:Unique Net Id Structure)
   > "SetText (Text)"(self=Get TextPlayerName() -> (TextPlayerName:Common Text Object Reference), InText=@n2.PlayerName)
   n4 = IsValid(Object=@n2.HeroItem)
   > SetActiveWidgetIndex(
    self=Get SwitcherConnectingStranger() -> (SwitcherConnectingStranger:Common Widget Switcher Legacy Object Reference)
    Index=Select(Option 0:int=1, Option 1:int=0, Index=@n4) -> (ReturnValue:int)
   )
   > Branch(Condition=@n4)
   > SetPlayerUniqueId!(
    execute
    self:Fort Mic Indicator Widget Object Reference=Get MicIndicator() -> (MicIndicator:Mic Indicator Object Reference)
    InPlayerUniqueId:"Unique Net Id Structure (by ref)"=Get PlayerNetId() -> (PlayerNetId:Unique Net Id Structure)
   ) -> (then)
   > SetVisibility(InVisibility=Select(
     Option 0:ESlateVisibility Enum=Collapsed
     Option 1:ESlateVisibility Enum=SelfHitTestInvisible
     Index=Break Fort Home Base Info!(FortHomeBaseInfo:"Fort Home Base Info Structure (by ref)"=@n2.HomeBaseInfo) -> (BannerIconId:String, BannerColorId:String, Name:Text, ValidData:bool=~) {is=title K2Node_BreakStruct}.ValidData
    ) -> (ReturnValue:ESlateVisibility Enum))
  ]}
]
variables [MemberInfo {
  type=FortTeamMemberInfo
  default {
   bBattlePassPurchased=false
   bHasBoostXp=false
   bHasRestXp=false
   bIsInZone=false
   bPartyLeader=false
   bannerColorId=""
   bannerIconId=""
   battlePassLevel=0
   heroClass=""
   heroItem=null
   heroLevel=""
   heroXP=-1
   numPlayersInParty=0
   playerIndex=-1
   playerName=""
   teamAffiliation=255
   heroIcon {
    bIsDynamicallyLoaded=false
    drawAs=Image
    imageType=NoImage
    mirroring=NoMirror
    resourceName=None
    resourceObject=null
    tiling=NoTile
    imageSize=<32 32>
    margin=<ltrb 0 0 0 0>
    outlineSettings {
     bUseBrushTransparency=false
     roundingType=HalfHeightRadius
     width=0
     color {colorUseRule=UseColor_Specified, specifiedColor=<rgba 0 0 0 0>}
     cornerRadii=<0 0 0 1>
    }
    tintColor {colorUseRule=UseColor_Specified, specifiedColor=<rgba 1 1 1 1>}
    uVRegion {bIsValid=false, max=<0 0>, min=<0 0>}
   }
   homeBaseInfo {name="", bannerColorId="", bannerIconId="", validData=false}
   selectedGadgetItems []
  }
 }]
widgets {
 info {inheritedWidgetCount=0, parentClass=&CommonUserWidget, rootWidgetClass=&SizeBox, widgetCount=12}
 widgets [
  SizeBox_9 {widgetClassPath=&SizeBox, widget=SizeBox_9, properties {heightOverride=24}}
  SwitcherConnectingStranger {
   widgetClassPath=&CommonWidgetSwitcherLegacy
   bIsVariable=true
   parent=SizeBox_9
   widget=SwitcherConnectingStranger
   slot {class=SizeBoxSlot}
  }
  HorizontalBox_1111 {widgetClassPath=&HorizontalBox, parent=SwitcherConnectingStranger, widget=HorizontalBox_1111, slot {class=WidgetSwitcherSlot}}
  CommonBorder_58 {
   widgetClassPath=&CommonBorder
   parent=HorizontalBox_1111
   widget=CommonBorder_58
   properties {
    background {
     bIsDynamicallyLoaded=false
     drawAs=NoDrawType
     imageType=NoImage
     mirroring=NoMirror
     resourceName=None
     resourceObject=null
     tiling=NoTile
     imageSize=<32 32>
     margin=<ltrb 0.4 0.4 0.4 0.4>
     outlineSettings {
      bUseBrushTransparency=false
      roundingType=HalfHeightRadius
      width=0
      color {colorUseRule=UseColor_Specified, specifiedColor=<rgba 0 0 0 0>}
      cornerRadii=<0 0 0 0>
     }
     tintColor {colorUseRule=UseColor_Specified, specifiedColor=<rgba 1 1 1 1>}
     uVRegion {bIsValid=false, max=<0 0>, min=<0 0>}
    }
    padding=<ltrb 0 0 0 0>
   }
   slot {class=HorizontalBoxSlot, properties {horizontalAlignment=HAlign_Left, verticalAlignment=VAlign_Center, padding=<ltrb 2 0 0 0>}}
  }
  HorizontalBox_57 {
   widgetClassPath=&HorizontalBox
   parent=CommonBorder_58
   widget=HorizontalBox_57
   slot {class=BorderSlot, properties {padding=<ltrb 0 0 0 0>}}
  }
  VerticalBox_5 {widgetClassPath=&VerticalBox, parent=HorizontalBox_57, widget=VerticalBox_5, slot {class=HorizontalBoxSlot}}
  HorizontalBox_1107 {
   widgetClassPath=&HorizontalBox
   parent=VerticalBox_5
   widget=HorizontalBox_1107
   slot {class=VerticalBoxSlot, properties {verticalAlignment=VAlign_Center}}
  }
  SizeBox_0 {
   widgetClassPath=&SizeBox
   parent=HorizontalBox_1107
   widget=SizeBox_0
   properties {bOverride_WidthOverride=true, widthOverride=32}
   slot {class=HorizontalBoxSlot}
  }
  MicIndicator {
   widgetClassPath=/Game/UI/InGame/HUD/MicIndicator/MicIndicator
   bIsVariable=true
   parent=SizeBox_0
   widget=MicIndicator
   slot {class=SizeBoxSlot}
  }
  TextPlayerName {
   widgetClassPath=&CommonTextBlock
   bIsVariable=true
   parent=HorizontalBox_1107
   widget=TextPlayerName
   properties {
    style=/FortUILibrary/Styles/Typography/Foundation/TextStyle-Base-XS-B-S.TextStyle-Base-XS-B-S_C
    text=Player Name
    font {
     bForceMonospaced=false
     bMaterialIsStencil=false
     fontMaterial=null
     fontObject=&NotoSans
     letterSpacing=0
     monospacedWidth=1
     size=16
     skewAmount=0
     typefaceFontName=NotoSans-Bold
     outlineSettings {
      bApplyOutlineToDropShadows=true
      bMiteredCorners=false
      bSeparateFillAlpha=true
      outlineMaterial=&M-FontShadow
      outlineSize=2
      outlineColor=<rgba 0 0 0 0.5>
     }
    }
    shadowOffset=<0 0>
   }
   slot {class=HorizontalBoxSlot, properties {horizontalAlignment=HAlign_Left, verticalAlignment=VAlign_Center}}
  }
  "[Horizontal Box] Connecting Stranger" {
   widgetClassPath=&HorizontalBox
   parent=SwitcherConnectingStranger
   widget="[Horizontal Box] Connecting Stranger"
   slot {class=WidgetSwitcherSlot}
  }
  TextBlock_6 {
   widgetClassPath=&CommonTextBlock
   bIsVariable=true
   parent="[Horizontal Box] Connecting Stranger"
   widget=TextBlock_6
   properties {
    style=&TextStyle-Base-XS-B-S_C
    text=Connecting...
    font {
     bForceMonospaced=false
     bMaterialIsStencil=false
     fontMaterial=null
     fontObject=&NotoSans
     letterSpacing=0
     monospacedWidth=1
     size=16
     skewAmount=0
     typefaceFontName=NotoSans-Bold
     outlineSettings {
      bApplyOutlineToDropShadows=true
      bMiteredCorners=false
      bSeparateFillAlpha=true
      outlineMaterial=&M-FontShadow
      outlineSize=2
      outlineColor=<rgba 0 0 0 0.5>
     }
    }
    shadowOffset=<0 0>
   }
   slot {class=HorizontalBoxSlot, properties {horizontalAlignment=HAlign_Left, verticalAlignment=VAlign_Top}}
  }
 ]
}
